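# Material function for anime fly-away hair strands, outputting lit and shaded colors, normal, world position offset, pixel depth offset and a DFL center-line mask.
# Spline thickening, with width chosen by vertex color green, supplies strand normals and offset, plus a root-to-tip push along vertex normals.
# Color uses a root-tip gradient or input map, modulated by tiled hair texture.
# Thickened normals blend with texture normals; temporal dither drives pixel depth offset.
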MaterialFunction /Game/Global/Materials/Character_Parent_2023/Hair_Parent_2023/MaterialFunctions/MF_FlyAwayHair_Anime_2026
graph {nodes [
  n1 = Distance(A=ComponentMask(None=TextureCoordinate()) {r=true}, B=Constant() {r=0.5})
  n2 = Constant() {r=0.5}
  n3 = ScalarParameter() {defaultValue=0.25, group=Anime FlyAwayHair, parameterName=FlyAwayHair_DFLineThickness}
  n4 = ScalarParameter() {defaultValue=2, group=Anime FlyAwayHair, parameterName=FlyAwayHair_DFLineHardness}
  FunctionOutput(None=NamedRerouteDeclaration(None=SphereMask(A=@n1, B=@n2, Radius=@n3, Hardness=@n4) {attenuationRadius=0.2, hardnessPercent=1}) {name=DFL, nodeColor=<rgba 1 0 0.2578 1>}) {outputName=DFL, sortPriority=10}
  n5 = Multiply(A=NamedRerouteUsage(), B=ScalarParameter() {defaultValue=0.5, group=Anime FlyAwayHair, parameterName=Anime_FlyAwayHair_TextureOpacity})
  n6 = NamedRerouteDeclaration(None=Multiply(A=NamedRerouteUsage(), B=Add(A=@n5))) {name=FinalLitColor, nodeColor=<rgba 0 0.8125 1 1>}
  FunctionOutput(None=@n6) {outputName=LitColor, sortPriority=1}
  n7 = VectorParameter() {group=Anime FlyAwayHair, parameterName=Anime_FlyAwayHair_ShadedDarkness, defaultValue=<rgba 0.6 0.6 0.6 1>}
  FunctionOutput(None=Multiply(A=@n6, B=@n7.RGB)) {outputName=ShadedColor, sortPriority=2}
  n8 = VertexColor()
  n9 = LinearInterpolate(
   A=ScalarParameter() {defaultValue=0.2, group=Anime FlyAwayHair, parameterName=SplineThicken_WidthBase_Min, sortPriority=1}
   B=ScalarParameter() {defaultValue=0.15, group=Anime FlyAwayHair, parameterName=SplineThicken_WidthBase_Middle, sortPriority=1}
   Alpha=Saturate(None=@n8.G)
  )
  n10 = If(
   A=@n8.G
   "A > B"=ScalarParameter() {defaultValue=0.25, group=Anime FlyAwayHair, parameterName=SplineThicken_WidthBase_ Max, sortPriority=1}
   "A == B"=@n9
   "A < B"=@n9
  ) {constB=0.7}
  n11 = MaterialFunctionCall!(
   widthTip=Multiply(A=ScalarParameter() {defaultValue=0.1, group=Anime FlyAwayHair, parameterName=SplineThicken_Width tip, sortPriority=1}, B=@n10)
   Flatten=ScalarParameter() {group=Anime FlyAwayHair, parameterName=SplineThicken_Flatten, sortPriority=1}
   widthBase=Multiply(A=@n10, B=ScalarParameter() {defaultValue=1, group=Anime FlyAwayHair, parameterName=Anime_FlyAwayHair_ThicknessMultiple})
  ) {materialFunction=&/MF_splineThicken_FlyawayHair}
  n12 = MaterialFunctionCall(
   BaseNormal=MaterialFunctionCall(
    Normal=Transform(None=@n11.normal) {transformSourceType=TRANSFORMSOURCE_World, transformType=TRANSFORM_Tangent}
    Flatness=ScalarParameter() {group=Anime FlyAwayHair, parameterName=SplineThicken_NormalFlatness, sortPriority=1}
   ) {materialFunction=&FlattenNormal}.Result
   AdditionalNormal=NamedRerouteUsage()
  ) {materialFunction=&BlendAngleCorrectedNormals}
  n13 = TwoSidedSign()
  FunctionOutput(None=NamedRerouteDeclaration(None=Multiply(A=@n12.Result, B=@n13)) {name=FinalNormal, nodeColor=<rgba 0.9453 0 1 1>}) {outputName=Normal, sortPriority=3}
  n14 = Power(Base=NamedRerouteUsage(), Exp=ScalarParameter() {defaultValue=1.5, group=Anime FlyAwayHair, parameterName=Anime_FlyAwayHair_RootDitherPower}) {constExponent=1.5}
  n15 = LinearInterpolate(B=ScalarParameter() {defaultValue=0.5, group=Anime FlyAwayHair, parameterName=Anime_FlyAwayHair_DitherAlphaThreshold}, Alpha=@n14)
  n16 = Constant() {r=1}
  n17 = ScalarParameter() {defaultValue=1.5, group=Anime FlyAwayHair, parameterName=Anime_FlyAwayHair_FurBlendOpacityMask}
  FunctionOutput(None=Multiply(
    A=MaterialFunctionCall(Alpha Threshold=@n15, Random=@n16) {materialFunction=&DitherTemporalAA}.Result
    B=LinearInterpolate(B=@n17, Alpha=@n14)
   )) {outputName=PixelDepthOffset}
  FunctionOutput(None=NamedRerouteUsage()) {outputName=WPO}
  n18 = VectorParameter() {group=Anime FlyAwayHair, parameterName=Anime_FlyAwayHair_LitColor, defaultValue=<rgba 0.6198 0.6198 0.6198 1>}
  n19 = VectorParameter() {group=Anime FlyAwayHair, parameterName=Anime_FlyAwayHair_LitColor_Tip, defaultValue=<rgba 0.4 0.4 0.4 1>}
  n20 = Power(
   Base=NamedRerouteUsage()
   Exp=ScalarParameter() {defaultValue=2, group=Anime FlyAwayHair, parameterName=Anime_FlyAwayHair_RootTipGradientPower}
  )
  NamedRerouteDeclaration(None=StaticSwitchParameter(True=FunctionInput() {inputName=ColorLit_In}, False=LinearInterpolate(A=@n18.RGB, B=@n19.RGB, Alpha=@n20)) {group=01_Anime Maps, parameterName=UseColorLitMap, sortPriority=0}) {name=Input LitColor, nodeColor=<rgba 1 0.6328 0 1>}
  n21 = MaterialFunctionCall(
   UVs=TextureCoordinate()
   Texture Scale=MaterialFunctionCall(X=ScalarParameter() {defaultValue=2, group=Anime FlyAwayHair, parameterName=Anime_FlyAwayHair_Tiling, sortPriority=2}, Y=Constant() {r=1}) {materialFunction=&MakeFloat2}.Result
  ) {materialFunction=&ScaleUVsByCenter}
  n22 = TextureSampleParameter2D(UVs=@n21.UVs) {
   group=Anime FlyAwayHair
   parameterName=FlyAwayHairNormal
   samplerSource=SSM_Wrap_WorldGroupSettings
   samplerType=SAMPLERTYPE_Normal
   texture=/Game/Global/Materials/Character_Parent_2023/Hair_Parent_2023/Texture/T_FlyAwayHair_N
  }
  NamedRerouteDeclaration(None=MaterialFunctionCall(Normal=@n22.RGB, Flatness=Constant() {r=0.5}) {materialFunction=&FlattenNormal}.Result) {name=HairTextureNormal, nodeColor=<rgba 1 0.1875 0 1>}
  n23 = VertexNormalWS()
  n24 = NamedRerouteDeclaration(None=ComponentMask(None=TextureCoordinate()) {desc="black is tip ,white is root", g=true}) {name=RootTip_Gradient color, nodeColor=<rgba 0.2422 0 1 1>}
  n25 = Multiply(A=NamedRerouteDeclaration(None=OneMinus(None=@n24)) {name=Reverse RootTip, nodeColor=<rgba 1 0 0.4922 1>}, B=NamedRerouteUsage())
  n26 = ScalarParameter() {defaultValue=0.15, group=Anime FlyAwayHair, parameterName=Anime_FlyAwayHair_TipWPO, sortPriority=2}
  n27 = NamedRerouteUsage()
  n28 = LinearInterpolate(
   A=ScalarParameter() {defaultValue=-0.05, group=Anime FlyAwayHair, parameterName=Anime_FlyAwayHair_RootWPO, sortPriority=2}
   B=ScalarParameter() {defaultValue=0.25, group=Anime FlyAwayHair, parameterName=Anime_FlyAwayHair_MidWPO, sortPriority=2}
   Alpha=Add(A=Multiply(A=@n25) {constB=4}, B=LinearInterpolate(B=@n26, Alpha=@n27) {constB=0})
  ) {constA=-0.03, constB=0.6}
  NamedRerouteDeclaration(None=Add(A=@n11.worldPositionOffset, B=Multiply(A=@n23, B=@n28))) {name=OutPutWPO, nodeColor=<rgba 0 1 0.625 1>}
  n29 = TextureSampleParameter2D(UVs=@n21.UVs) {
   group=Anime FlyAwayHair
   parameterName=FlyAwayHairTexture
   samplerSource=SSM_Wrap_WorldGroupSettings
   texture=/Game/Global/Materials/Character_Parent_2023/Hair_Parent_2023/Texture/T_FlyAwayHair_D
  }
  NamedRerouteDeclaration(None=LinearInterpolate(
    A=ScalarParameter() {group=Anime FlyAwayHair, parameterName=Anime_FlyAwayHair_Darkness}
    B=ScalarParameter() {defaultValue=1, group=Anime FlyAwayHair, parameterName=Anime_FlyAwayHair_Brightness}
    Alpha=@n29.R
   )) {name=FlyAwayHairtexture, nodeColor=<rgba 1 0 0.2109 1>}
 ]}
