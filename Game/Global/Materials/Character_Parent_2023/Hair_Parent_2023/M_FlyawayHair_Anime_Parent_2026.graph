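# Parent material for anime-style flyaway hair on skeletal meshes, using masked, two-sided Cloth shading.
# Chains the shared anime shader Pre, Calculations and Post functions, with Pre sampling UV1 as the normal hair layout.
# A flyaway-hair function takes the lit color and supplies color for both lit and shaded inputs, normal, extra DFL mask, WPO and Pixel Depth Offset.
# The shading gradient is inverted before Post, which the author's note marks as uncertain.
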Material /Game/Global/Materials/Character_Parent_2023/Hair_Parent_2023/M_FlyawayHair_Anime_Parent_2026
graph {
 properties {
  bAutomaticallySetUsageInEditor=false
  bUseMaterialAttributes=true
  bUsedWithSkeletalMesh=true
  blendMode=BLEND_Masked
  pixelDepthOffsetMode=PDOM_Legacy
  shadingModel=MSM_Cloth
  twoSided=true
  usedShadingModels=Cloth
 }
 nodes [
  n1 = MaterialFunctionCall!(CustomUV=TextureCoordinate() {coordinateIndex=1, desc=FlyAway hair UV1 is set to normal uv layout}) {materialFunction=/Game/Global/Materials/Character_Parent_2023/Anime_Parent_2023/MaterialFunctions/MFC_AnimeShader_2023_Pre}
  n2 = MaterialFunctionCall!(ColorLit_In=@n1.Color_Lit) {materialFunction=/Game/Global/Materials/Character_Parent_2023/Hair_Parent_2023/MaterialFunctions/MF_FlyAwayHair_Anime_2026}
  n3 = MaterialFunctionCall!(CustomLightVector, OverrideNormal_WS, STT=@n1.STT, Normal_Map=@n2.Normal, DFL=@n1.STT, FX=@n1.STT) {materialFunction=/Game/Global/Materials/Character_Parent_2023/Anime_Parent_2023/MaterialFunctions/MFC_AnimeShader_2023_Calculations}
  SetMaterialAttributes!(
   MaterialAttributes=MaterialFunctionCall!(
    Shading_Gradient=OneMinus(None=@n3.Shading_Gradient) {desc=Shading curve is reverted?}
    Threshold_Gradient=@n3.Threshold_Gradient
    Specular_Gradient=@n3.Threshold_Gradient
    Sheen_Mask=@n3.Threshold_Gradient
    Thickness=@n3.Threshold_Gradient
    DFL_Mask=Add(A=@n3.DFL_Mask, B=@n2.DFL)
    FX=@n3.Threshold_Gradient
    Metallic_Gradient=@n3.Threshold_Gradient
    RimLight_Gradient=@n3.Threshold_Gradient
    NormalBuffer=@n3.Threshold_Gradient
    --------
    CINE_LightSources=@n3.Threshold_Gradient
    CINE_LightColor=@n3.Threshold_Gradient
    ---------
    Color_Lit=@n2.LitColor
    Color_Shaded=@n2.LitColor
    SRM=@n1.SRM
    Emissive=@n1.SRM
    DFL_Color=@n1.SRM
    Specular_Color=@n1.SRM
    RimLight_Color=@n1.SRM
   ) {materialFunction=/Game/Global/Materials/Character_Parent_2023/Anime_Parent_2023/MaterialFunctions/MFC_AnimeShader_2023_Post}.Result
   World Position Offset=@n2.WPO
   Pixel Depth Offset=@n2.PixelDepthOffset
   Emissive Color
  ) {attributeSetTypes [#1, #2, #3]}
 ]
}
